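; Blueprint interface for audio gameplay volumes in the MusicEncounter content, exposing their cylinder dimensions to callers.
; It declares two functions without bodies: SetCylinderHalfHeight takes a half height and SetRadius takes a radius, both double-precision floats.
; Implementing actors decide how these values resize the volume; the interface itself holds no logic.
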(fn SetCylinderHalfHeight
  (params (Half%20Height :type Float%20%28double-precision%29)))

(fn SetRadius
  (params (Radius :type Float%20%28double-precision%29)))
