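; Audio gameplay volume Blueprint that tells a music encounter when the listener enters or leaves it.
; On toggle on and toggle off, it fetches its Music Encounter reference and branches on it, doing nothing on the else path.
; It casts the reference to the BPI_MusicEncounter_Volume interface and calls ListenerEntered or ListenerExited, leaving the response to the encounter.
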(event OnToggledOn(AudioGameplayVolume)
  (bind _music_encounter (Variables|MusicEncounter|GetMusic%20Encounter)
    (:then
      (bind _asmusic_encounter_volume (Utilities|Casting|CastToBPI_MusicEncounter_Volume _music_encounter))
      (MusicEncounter|Volume|ListenerEntered _asmusic_encounter_volume))
    (:else)))

(event OnToggledOff(AudioGameplayVolume)
  (bind _music_encounter (Variables|MusicEncounter|GetMusic%20Encounter)
    (:then
      (bind _asmusic_encounter_volume (Utilities|Casting|CastToBPI_MusicEncounter_Volume _music_encounter))
      (MusicEncounter|Volume|ListenerExited _asmusic_encounter_volume))
    (:else)))

(fn ConstructionScript)
