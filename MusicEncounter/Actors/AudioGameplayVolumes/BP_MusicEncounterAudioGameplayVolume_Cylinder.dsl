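; Cylinder variant of the MusicEncounter audio gameplay volume, giving the volume a cylindrical shape through an AudioShapeCylinder component.
; Two MusicEncounter volume events pass their values straight to that component, one setting the cylinder's half height and the other its radius.
; The construction script is empty, so this graph only resizes the shape when those events are called.
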(event MusicEncounter|Volume|EventSetCylinderHalfHeight
  (params (Half%20Height :type Float%20%28double-precision%29))
  (Control|SetHalfHeight (Variables|Default|GetAudioShapeCylinder) Half%20Height))

(event MusicEncounter|Volume|EventSetRadius
  (params (Radius :type Float%20%28double-precision%29))
  (Control|SetRadius (Variables|Default|GetAudioShapeCylinder) Radius :!class /Script/AudioShapes.AudioShapeCylinderComponent))

(fn ConstructionScript)
