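; Blueprint audio gameplay volume with a spherical shape, used in the music encounter system.
; Its Music Encounter volume event for setting radius passes the given radius straight to the AudioShapeSphere component.
; The construction script is empty.
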(event MusicEncounter|Volume|EventSetRadius
  (params (Radius :type Float%20%28double-precision%29))
  (Control|SetRadius (Variables|Default|GetAudioShapeSphere) Radius :!class /Script/AudioShapes.AudioShapeSphereComponent))

(fn ConstructionScript)
